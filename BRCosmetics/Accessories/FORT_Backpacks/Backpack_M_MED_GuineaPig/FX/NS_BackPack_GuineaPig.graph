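# Niagara effect for the guinea pig backpack cosmetic that throws short-lived, falling orange sparks from skeletal mesh sockets.
# Sparks_Shoudler fires from a random socket between long randomized delays, while Sparks_Wire uses a system-chosen socket.
# Both send first-frame location events with the socket index packed as a float.
# The glow emitter pins a brief camera-offset flash to the socket read from MaterialRandom, though no event handler is shown.
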NiagaraSystem /BRCosmetics/Accessories/FORT_Backpacks/Backpack_M_MED_GuineaPig/FX/NS_BackPack_GuineaPig
dependencies {
 usedDataInterfaces [&NiagaraDataInterfaceSkeletalMesh]
 usedDynamicInputs [&Lerp_Float]
 usedModules [
  &self:SystemSpawnScript.SystemScriptSource.SystemScriptGraph.NiagaraNodeAssignment_0.SetVariables_#1
  &SystemState
  &self:SystemSpawnScript.SystemScriptSource.SystemScriptGraph.NiagaraNodeAssignment_1.SetVariables_#2
 ]
}
emitters {
 Sparks_Shoudler {
  emitterUpdateScript [
   &EmitterState(
    Inactive Response=::NewEnumerator1
    Life Cycle Mode=::NewEnumerator1
    Loop Delay=&RandomRangeFloat(Minimum=15, Maximum=30)
    Loop Duration=&RandomRangeFloat(Minimum=40, Maximum=120)
    Recalculate Duration Each Loop=true
    UseLoopDelay=true
   )
   &EmitterState(
    Inactive Response=::NewEnumerator1
    Life Cycle Mode=::NewEnumerator1
    Loop Delay=&RandomRangeFloat(Minimum=15, Maximum=30)
    Loop Duration=&RandomRangeFloat(Minimum=40, Maximum=120)
    Recalculate Duration Each Loop=true
    UseLoopDelay=true
   ) {enabled=false, module=EmitterState001}
   &SpawnBurst_Instantaneous(
    Spawn Count=&RandomRangeInt(Minimum=3, Maximum=2)
    Spawn Probability=0.23
    Use Spawn Probability=false
    UseLoopCountLimit=false
   )
   &SpawnRate(
    SpawnRate=&Multiply_Float(
     A=&RandomRangeFloat(Minimum=5, Maximum=10)
     B=&FloatFromCurve(FloatCurve=DataInterface() {propertyValues {Curve {keys [<at 0 1>,<at 0.2 0>]}}}, CurveIndex=$Emitter.LoopedAge)
    )
    Use Spawn Probability=false
   ) {module=SpawnRate001}
   /SetVariables(Emitter.Socket:NiagaraInt32=&RandomRangeInt(Minimum=0, Maximum=2)) {isSetParameters=true}
  ]
  particleSpawnScript [
   &InitializeParticle(
    Color Mode=::NewEnumerator0
    Lifetime Max=0.5625
    Lifetime Min=0.375
    Lifetime Mode=::NewEnumerator1
    Sprite Size=&Multiply_Vector2D(
     A=&RandomRangeVector2D(Minimum=<0.35 2>, Maximum=<0.5 3>)
     B=&Vector2DFromFloat(Value=1)
    )
    Sprite Size Mode=::NewEnumerator4
    UsePositionOffset=false
    Write Material Random=true
   )
   &SkeletalMeshLocation(
    Bone / Socket Index=$Emitter.Socket
    MaskSpawnGroup=false
    Mesh Sampling Type=::NewEnumerator5
    Sampled Position Offset=<0 0 0>
    Skeletal Mesh=$System.Skeletal Mesh
    Socket Sampling Mode=::NewEnumerator5
    Use Sampled Mesh Attribute as Ribbon ID=false
   )
   &ShapeLocation(MaskSpawnGroup=false, Override Seed=false, Sphere Radius=1, UseOwnerScale=true)
   &AddVelocity(
    Cone Axis=&ReturnMeshOrientationAxis(Mesh Orientation Axis=::XAxis, Mesh Orientation Quat=$Particles.SkeletalMeshLocation.SampledRotation)
    Module.UseConeFalloff=false
    Module.UseVelDistribution=false
    Override Seed=false
    Velocity Mode=::NewEnumerator2
    Velocity Speed=&RandomRangeFloat(Minimum=75, Maximum=200)
   )
  ]
  particleUpdateScript [
   &ParticleState()
   /Niagara/Modules/Update/Forces/CurlNoiseForce(Noise Frequency=0.1, Noise Strength=250, Pan Noise=true, Pan Noise Field=<0 0 1>) {enabled=false}
   &Drag(
    Drag=&FloatFromCurve(
     FloatCurve=DataInterface() {propertyValues {Curve {keys [{interpMode=RCIM_Cubic, tangentMode=RCTM_User, time=0, value=0}, {interpMode=RCIM_Cubic, tangentMode=RCTM_User, time=1, value=1}]}}}
     Scale Curve=6
    )
    Use Linear Drag=true
    Use Rotational Drag=true
   )
   &Color(
    Color=&ColorFromCurve(ColorCurve=DataInterface() {propertyValues {
       AlphaCurve {keys [
         {interpMode=RCIM_Cubic, leaveTangentWeight=0.3, tangentMode=RCTM_User, time=0, value=1}
         {arriveTangentWeight=0.4243, interpMode=RCIM_Cubic, leaveTangentWeight=0.0471, tangentMode=RCTM_User, time=0.9, value=1}
         {arriveTangent=-25.6518, arriveTangentWeight=0.0333, interpMode=RCIM_Cubic, leaveTangent=-25.6518, tangentMode=RCTM_User, time=1, value=0}
        ]}
       BlueCurve {keys [
         <at 0.1461 0.3646 RCIM_Cubic>
         {arriveTangent=-0.436, interpMode=RCIM_Cubic, leaveTangent=-0.436, time=0.3452, value=0.1719}
         <at 0.9823 0 RCIM_Cubic>
        ]}
       GreenCurve {keys [
         {arriveTangent=-46.6396, interpMode=RCIM_Cubic, leaveTangent=-46.6396, tangentMode=RCTM_User, time=0.1461, value=6.2133}
         {arriveTangent=-7.2714, interpMode=RCIM_Cubic, leaveTangent=-7.2714, time=0.3452, value=1.3872}
         <at 0.9823 0.1333 RCIM_Cubic>
        ]}
       RedCurve {keys [
         {arriveTangent=-46.6396, interpMode=RCIM_Cubic, leaveTangent=-46.6396, tangentMode=RCTM_User, time=0.1461, value=10}
         {arriveTangent=-10.7637, interpMode=RCIM_Cubic, leaveTangent=-10.7637, time=0.3452, value=3}
         <at 0.9823 1 RCIM_Cubic>
        ]}
      }})
    Scale Color=&VectorFromFloat(Value=&Multiply_Float(A=&RandomRangeFloat(Minimum=1, Maximum=3), B=1))
   ) {module=Color001}
   &ScaleColor(
    Color Value To Scale=$Particles.Color
    Scale Alpha=&Multiply_Float(
     A=&FloatFromCurve(FloatCurve=DataInterface() {propertyValues {Curve {keys [
          {arriveTangent=16.6122, interpMode=RCIM_Cubic, leaveTangent=16.6122, tangentMode=RCTM_User, time=0, value=0}
          <at 0.1 1 RCIM_Cubic>
          <at 1 0 RCIM_Cubic>
         ]}}})
     B=&Multiply_Float(A=1, B=&FNTimeOfDayScalar(Night Scale=0.7, Morning Scale=1, Evening Scale=1))
    )
    ScaleA=true
    ScaleRGB=true
   )
   &GravityForce(Gravity=<0 0 -680>) {module=GravityForce001}
   &InheritSourceMovement(
    Applied Position Scale=&VectorFromFloat(Value=&FloatFromCurve(FloatCurve=DataInterface() {propertyValues {Curve {keys [<at 0 1>,<at 1 0>]}}}))
    ApplyForce=false
    ApplyPosition=true
    ApplyPositionFalloff=false
   ) {enabled=false}
   &SolveForcesAndVelocity(Clamp Velocity=false, Limit Acceleration=false)
   /SetVariables(Particles.Bonetransfer:NiagaraFloat=&MakeFloatFromInt(Int=$Emitter.Socket)) {isSetParameters=true}
   &GenerateLocationEvent(
    Event Generation Enabled:NiagaraBool=$Transient.FirstFrame
    Event Send Rate=120
    "Float to Send as Float 2 (RandomNormalizedFloat)":NiagaraFloat=$Particles.Bonetransfer
    Use Event Probability=false
    UseEventDelay=false
   )
   &ScaleSpriteSizeBySpeed(Max Scale Factor=<1 2>)
  ]
  rendererClasses [&NiagaraSpriteRendererProperties]
  renderers [{rendererClass=&NiagaraSpriteRendererProperties}]
 }
 Sparks_Wire {
  emitterUpdateScript [
   &EmitterState()
   &EmitterState(
    Inactive Response=::NewEnumerator1
    Life Cycle Mode=::NewEnumerator1
    Loop Delay=1
    Loop Duration=&RandomRangeFloat(Minimum=2, Maximum=6)
    Recalculate Duration Each Loop=true
    UseLoopDelay=false
   ) {enabled=false, module=EmitterState001}
   &SpawnBurst_Instantaneous(
    Spawn Count=&RandomRangeInt(Minimum=3, Maximum=5)
    Spawn Probability=0.23
    Use Spawn Probability=false
    UseLoopCountLimit=false
   )
   &SpawnBurst_Instantaneous(
    Spawn Count=&RandomRangeInt(Minimum=3, Maximum=2)
    Spawn Probability=0.1
    Spawn Time=0.15
    Use Spawn Probability=true
    UseLoopCountLimit=false
   ) {module=SpawnBurst_Instantaneous001}
   &SpawnRate(
    SpawnRate=&Multiply_Float(
     A=&RandomRangeFloat(Minimum=5, Maximum=10)
     B=&FloatFromCurve(FloatCurve=DataInterface() {propertyValues {Curve {keys [<at 0 1>,<at 0.2 0>]}}}, CurveIndex=$Emitter.LoopedAge)
    )
    Use Spawn Probability=false
   ) {module=SpawnRate001}
   /SetVariables(Emitter.Socket:NiagaraInt32=$System.Socket) {isSetParameters=true}
  ]
  particleSpawnScript [
   &InitializeParticle(
    Color Mode=::NewEnumerator0
    Lifetime Max=0.375
    Lifetime Min=0.1
    Lifetime Mode=::NewEnumerator1
    Sprite Size=&Multiply_Vector2D(
     A=&RandomRangeVector2D(Minimum=<0.25 1>, Maximum=<0.85 5>)
     B=&Vector2DFromFloat(Value=1)
    )
    Sprite Size Mode=::NewEnumerator4
    UsePositionOffset=false
    Write Material Random=true
   )
   &SkeletalMeshLocation(
    Bone / Socket Index=$Emitter.Socket
    MaskSpawnGroup=false
    Mesh Sampling Type=::NewEnumerator5
    Orientation Sampling=::NewEnumerator1
    Skeletal Mesh=$System.Skeletal Mesh
    Socket Sampling Mode=::NewEnumerator5
    Use Sampled Mesh Attribute as Ribbon ID=false
   )
   &ShapeLocation(MaskSpawnGroup=false, Override Seed=false, Sphere Radius=1, UseOwnerScale=true)
   &AddVelocity(
    Cone Axis=&ReturnMeshOrientationAxis(Mesh Orientation Axis=::XAxis, Mesh Orientation Quat=$Particles.SkeletalMeshLocation.SampledRotation)
    Module.UseConeFalloff=false
    Module.UseVelDistribution=false
    Override Seed=false
    Velocity Mode=::NewEnumerator2
    Velocity Speed=&RandomRangeFloat(Minimum=75, Maximum=200)
   )
  ]
  particleUpdateScript [
   &ParticleState()
   &ScaleSpriteSize(
    Scale Sprite Size Mode=::NewEnumerator0
    Uniform Scale Factor=&FloatFromCurve(FloatCurve=DataInterface() {propertyValues {Curve {keys [<at 0.75 1 RCIM_Cubic>,<at 1 0 RCIM_Cubic>]}}}, Scale Curve=1.35)
   )
   &Color(
    Color=&ColorFromCurve(ColorCurve=DataInterface() {propertyValues {
       AlphaCurve {keys [
         {interpMode=RCIM_Cubic, leaveTangentWeight=0.3, tangentMode=RCTM_User, time=0, value=1}
         {arriveTangentWeight=0.4243, interpMode=RCIM_Cubic, leaveTangentWeight=0.0471, tangentMode=RCTM_User, time=0.9, value=1}
         {arriveTangent=-25.6518, arriveTangentWeight=0.0333, interpMode=RCIM_Cubic, leaveTangent=-25.6518, tangentMode=RCTM_User, time=1, value=0}
        ]}
       BlueCurve {keys [
         <at 0.1461 0.3646 RCIM_Cubic>
         {arriveTangent=-0.436, interpMode=RCIM_Cubic, leaveTangent=-0.436, time=0.3452, value=0.1719}
         <at 0.9823 0 RCIM_Cubic>
        ]}
       GreenCurve {keys [
         {arriveTangent=-46.6396, interpMode=RCIM_Cubic, leaveTangent=-46.6396, tangentMode=RCTM_User, time=0.1461, value=6.2133}
         {arriveTangent=-7.2714, interpMode=RCIM_Cubic, leaveTangent=-7.2714, time=0.3452, value=1.3872}
         <at 0.9823 0.1333 RCIM_Cubic>
        ]}
       RedCurve {keys [
         {arriveTangent=-46.6396, interpMode=RCIM_Cubic, leaveTangent=-46.6396, tangentMode=RCTM_User, time=0.1461, value=10}
         {arriveTangent=-10.7637, interpMode=RCIM_Cubic, leaveTangent=-10.7637, time=0.3452, value=3}
         <at 0.9823 1 RCIM_Cubic>
        ]}
      }})
    Scale Color=&VectorFromFloat(Value=&Multiply_Float(A=&RandomRangeFloat(Minimum=1, Maximum=3), B=1))
   ) {module=Color001}
   &CurlNoiseForce(Noise Frequency=0.1, Noise Strength=250, Pan Noise=true, Pan Noise Field=<0 0 1>) {enabled=false}
   &Drag(Drag=&RandomRangeFloat(Minimum=4, Maximum=10), Use Linear Drag=true, Use Rotational Drag=false)
   &ScaleColor(
    Color Value To Scale=$Particles.Color
    Scale Alpha=&Multiply_Float(
     A=&FloatFromCurve(FloatCurve=DataInterface() {propertyValues {Curve {keys [
          {arriveTangent=16.6122, interpMode=RCIM_Cubic, leaveTangent=16.6122, tangentMode=RCTM_User, time=0, value=0}
          {
           arriveTangentWeight=0.0333
           interpMode=RCIM_Cubic
           leaveTangentWeight=0.7554
           tangentMode=RCTM_Break
           tangentWeightMode=RCTWM_WeightedBoth
           time=0.1
           value=1
          }
          {
           arriveTangent=-7.511
           arriveTangentWeight=0.9523
           interpMode=RCIM_Cubic
           leaveTangent=-7.511
           leaveTangentWeight=0.9523
           tangentMode=RCTM_User
           tangentWeightMode=RCTWM_WeightedBoth
           time=1
           value=0
          }
         ]}}})
     B=&Multiply_Float(A=500, B=&FNTimeOfDayScalar(Night Scale=0.7, Morning Scale=1, Evening Scale=1))
    )
    ScaleA=true
    ScaleRGB=true
   ) {enabled=false}
   &GravityForce(Gravity=<0 0 -680>) {module=GravityForce001}
   &InheritSourceMovement(
    Applied Position Scale=&VectorFromFloat(Value=&FloatFromCurve(FloatCurve=DataInterface() {propertyValues {Curve {keys [<at 0 1 RCIM_Cubic>, {arriveTangent=-3.1142, tangentMode=RCTM_User, time=1, value=0}]}}}))
    ApplyForce=false
    ApplyPosition=true
    ApplyPositionFalloff=false
   ) {enabled=false}
   &SolveForcesAndVelocity(Clamp Velocity=false, Limit Acceleration=false)
   /SetVariables(Particles.Bonetransfer:NiagaraFloat=&MakeFloatFromInt(Int=$Emitter.Socket)) {isSetParameters=true}
   &GenerateLocationEvent(
    Event Generation Enabled:NiagaraBool=$Transient.FirstFrame
    Event Send Rate=120
    "Float to Send as Float 2 (RandomNormalizedFloat)":NiagaraFloat=$Particles.Bonetransfer
    Use Event Probability=false
    UseEventDelay=false
   )
   &ScaleSpriteSizeBySpeed(Max Scale Factor=<1 2>, Velocity Threshold=800)
  ]
  rendererClasses [&NiagaraSpriteRendererProperties]
  renderers [{rendererClass=&NiagaraSpriteRendererProperties}]
 }
 initail_Spark_Glow {
  emitterUpdateScript [&EmitterState()]
  particleSpawnScript [
   &InitializeParticle(
    Color=<rgba 10 7.249 3 1>
    Lifetime=0.27
    Position=<0 0 -5000000>
    Position Mode=::NewEnumerator1
    Sprite Size Mode=::NewEnumerator3
    Uniform Sprite Size=25
    UsePositionOffset=true
    Write Material Random=true
   )
   &CameraOffset(Camera Offset Amount=12)
  ]
  particleUpdateScript [
   &ParticleState()
   &SkeletalMeshLocation(
    Bone / Socket Index=&MakeIntFromFloat(Float=$Particles.MaterialRandom)
    MaskSpawnGroup=false
    Mesh Sampling Type=::NewEnumerator5
    Orientation Sampling=::NewEnumerator1
    Skeletal Mesh=$System.Skeletal Mesh
    Socket Sampling Mode=::NewEnumerator5
    Use Sampled Mesh Attribute as Ribbon ID=false
   )
   &DynamicMaterialParameters(
    Index 0 Param 1=0.8
    Index 0 Param 2=1
    Index 0 Param 3=1
    Index 1 Param 1=25
    Index 1 Param 4=1
    Module.Param0WriteEnabled=true
    Module.Param0WriteEnabled1=true
    Module.Param1WriteEnabled=true
    Module.Param1WriteEnabled1=true
    Module.Param2WriteEnabled=true
    Module.Param2WriteEnabled1=true
    Module.Param3WriteEnabled=false
    Module.Param3WriteEnabled1=true
    Write Parameter Index 1=true
   )
   &ScaleSpriteSize(Uniform Curve Sprite Scale=DataInterface() {propertyValues {Curve {keys [
        {arriveTangent=9.0725, interpMode=RCIM_Cubic, leaveTangent=9.0725, tangentMode=RCTM_User, time=0.05, value=0.5}
        {arriveTangentWeight=0.0621, interpMode=RCIM_Cubic, leaveTangentWeight=0.2957, tangentMode=RCTM_User, time=0.2149, value=1}
        {arriveTangent=-0.0405, interpMode=RCIM_Cubic, leaveTangent=-0.0405, tangentMode=RCTM_User, time=1, value=0}
       ]}}})
   &ScaleColor(
    Scale Alpha=&FloatFromCurve(FloatCurve=DataInterface() {propertyValues {Curve {keys [
         {arriveTangent=9.0725, interpMode=RCIM_Cubic, leaveTangent=9.0725, tangentMode=RCTM_User, time=0, value=0}
         {arriveTangent=-2.5249, interpMode=RCIM_Cubic, leaveTangent=-2.5249, tangentMode=RCTM_User, time=0.1408, value=0.9847}
         {arriveTangent=-0.0405, interpMode=RCIM_Cubic, leaveTangent=-0.0405, tangentMode=RCTM_User, time=1, value=0}
        ]}}})
    ScaleA=true
    ScaleRGB=true
   )
  ]
  rendererClasses [&NiagaraSpriteRendererProperties]
  renderers [
   {rendererClass=&NiagaraSpriteRendererProperties}
   {rendererClass=&NiagaraSpriteRendererProperties, rendererIndex=1}
  ]
 }
}
summary {
 systemName=NS_BackPack_GuineaPig
 emitters [
  {emitterName=Sparks_Shoudler, rendererClasses [&NiagaraSpriteRendererProperties]}
  {emitterName=Sparks_Wire, rendererClasses [&NiagaraSpriteRendererProperties]}
  {emitterName=initail_Spark_Glow, rendererClasses [&NiagaraSpriteRendererProperties]}
 ]
 userVariables [
  User.BaseCurveColor {
   type=NiagaraDataInterfaceColorCurve
   description=""
   defaultValue {
    struct=NiagaraExt_VariableValue_DataInterface
    value {
     dataInterface=&self:NiagaraDataInterfaceColorCurve_0
     dataInterfaceClass=&NiagaraDataInterfaceColorCurve
    }
   }
  }
  User.BaseCurveColorHueShift {
   type=NiagaraDataInterfaceCurve
   description=""
   defaultValue {
    struct=NiagaraExt_VariableValue_DataInterface
    value {
     dataInterface=&self:NiagaraDataInterfaceCurve_0
     dataInterfaceClass=&NiagaraDataInterfaceCurve
    }
   }
  }
  User.BaseCurveColorPick {type=NiagaraFloat, defaultValue=0, description=""}
 ]
}
