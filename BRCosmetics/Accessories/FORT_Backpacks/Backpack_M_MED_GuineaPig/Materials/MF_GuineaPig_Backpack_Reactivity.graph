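# Material function for the GuineaPig backpack that drives a timed emissive glow reaction.
# A game-time parameter plus a duration forms a countdown against Time, fading incoming emissive intensity from max to min glow through SmoothStep.
# The same countdown shifts the base of a random flicker multiplier before emissive is written back into the material attributes.
# A square mask generator call is present but unconnected to the output.
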MaterialFunction /BRCosmetics/Accessories/FORT_Backpacks/Backpack_M_MED_GuineaPig/Materials/MF_GuineaPig_Backpack_Reactivity
graph {nodes [
  n1 = GetMaterialAttributes(None=FunctionInput() {inputName=MaterialAttributes, inputType=FunctionInput_MaterialAttributes}) {attributeGetTypes [#1]}
  n2 = ScalarParameter() {defaultValue=2.2, group=GuineaPig_Reactivity, parameterName=ReactivityDuration}
  n3 = Add(A=ScalarParameter() {defaultValue=-9999, group=GuineaPig_Reactivity, parameterName=BP_GetGameTime, sortPriority=10000}, B=@n2)
  n4 = Time()
  n5 = Saturate(None=Divide(A=Subtract(A=@n3, B=@n4), B=@n2))
  n6 = Multiply(
   A=@n1.EmissiveColor
   B=LinearInterpolate(
    A=ScalarParameter() {defaultValue=0.35, group=GuineaPig_Reactivity, parameterName=Glow_Min, sortPriority=10}
    B=ScalarParameter() {defaultValue=2.5, group=GuineaPig_Reactivity, parameterName=Glow_Max, sortPriority=10}
    Alpha=SmoothStep(Value=@n5) {constMax=0.56}
   ) {constA=0.6, constB=5}
  )
  n7 = SmoothStep(Value=@n5) {constMax=0.5}
  n8 = Saturate(None=Add(
    A=LinearInterpolate(Alpha=@n7) {constA=0.8, constB=0.65}
    B=MaterialFunctionCall() {materialFunction=/Game/MF__FloatFlickerRandom_Generic}.RandomFlickerFloat
   ) {constA=0.7})
  FunctionOutput(None=SetMaterialAttributes!(MaterialAttributes=@n1.MaterialAttributes, EmissiveColor=Multiply(A=@n6, B=@n8)) {attributeSetTypes [#1]})
  MaterialFunctionCall!(
   Texture Coordinate=TextureCoordinate()
   X Pos Right=Constant() {r=0.8}
   X Pos Left=Constant() {r=0.6}
   Y Pos Up=Constant() {r=0.08}
   Y Pos Down=Constant() {r=0.17}
   SquareSoftness
   RotateSquare
   RotationAmount
  ) {materialFunction=/Game/Accessories/FORT_Backpacks/Backpack_F_MED_GroovyReader_Play/Materials/MaterialFunctions/MF_SquareMaskGenerator}
 ]}
