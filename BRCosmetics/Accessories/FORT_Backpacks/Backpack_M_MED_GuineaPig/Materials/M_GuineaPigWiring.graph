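# Masked material for the wires on the GuineaPig backpack cosmetic, showing colored insulation with bare metallic wire ends.
# SplineThicken expands the wire geometry through world position offset and supplies its normal, with width driven by vertex blue.
# Vertex blue also picks insulation color from a lookup texture, shaded by a UV gradient and tiled roughness texture.
# UV steps mask a narrower inner-wire tip with added normal detail, blended over the insulation as material attributes.
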Material /BRCosmetics/Accessories/FORT_Backpacks/Backpack_M_MED_GuineaPig/Materials/M_GuineaPigWiring
graph {
 properties {
  bUseMaterialAttributes=true
  bUsedWithMeshDeformer=true
  bUsedWithSkeletalMesh=true
  blendMode=BLEND_Masked
  pixelDepthOffsetMode=PDOM_Legacy
  usedShadingModels=Default Lit
 }
 nodes [
  Constant() {r=1}
  n1 = TextureCoordinate()
  n2 = ScalarParameter() {defaultValue=0.05, parameterName=EndWireLengthMin}
  n3 = ScalarParameter() {defaultValue=0.15, parameterName=EndWireLengthMax}
  n4 = NamedRerouteUsage()
  NamedRerouteDeclaration(None=Step(Y=ComponentMask(None=@n1) {g=true}, X=LinearInterpolate(A=@n2, B=@n3, Alpha=@n4)) {constX=0.5, constY=0.5}) {name=WireEndBlend, nodeColor=<rgba 0 0.4609 1 1>}
  n5 = VectorParameter() {parameterName=RoughnessTile, defaultValue=<rgba 0.2 2 0 1>}
  n6 = TextureSample(UVs=Multiply(A=TextureCoordinate(), B=ComponentMask(None=@n5.RGB) {g=true, r=true})) {samplerType=SAMPLERTYPE_LinearColor, texture=/BRCosmetics/Accessories/FORT_Backpacks/Backpack_M_MED_GuineaPig/Textures/T_Wires_S}
  NamedRerouteDeclaration(None=LinearInterpolate(
    A=ScalarParameter() {defaultValue=0.2, parameterName=MinRoughness}
    B=ScalarParameter() {defaultValue=0.5, parameterName=MaxRoughness}
    Alpha=@n6.R
   )) {name=Roughness, nodeColor=<rgba 0 0.3906 1 1>}
  n7 = ScalarParameter() {defaultValue=0.5, parameterName=RoughnessBlendToDiffuse}
  n8 = OneMinus(None=ComponentMask(None=TextureCoordinate()) {g=true})
  n9 = VertexColor()
  n10 = TextureSample(UVs=AppendVector(A=NamedRerouteDeclaration(None=@n9.B) {name=VertColorBlue, nodeColor=<rgba 0.8984 1 0 1>}, B=Constant() {r=0.5})) {texture=/BRCosmetics/Accessories/FORT_Backpacks/Backpack_M_MED_GuineaPig/Textures/T_GuineaPig_Wires}
  NamedRerouteDeclaration(None=Multiply(A=LinearInterpolate(B=@n6.R, Alpha=@n7) {constA=1}, B=Multiply(A=@n8, B=@n10.RGB))) {name=WireBaseColor, nodeColor=<rgba 0.3125 1 0 1>}
  n11 = LinearInterpolate(
   A=ScalarParameter() {defaultValue=0.4, parameterName=WireWidthMin}
   B=ScalarParameter() {defaultValue=0.675, parameterName=WireWidthMax}
   Alpha=VertexColor().B
  )
  n12 = TextureCoordinate()
  n13 = AppendVector(A=OneMinus(None=ComponentMask(None=@n12) {r=true}), B=ComponentMask(None=@n12) {g=true})
  n14 = MaterialFunctionCall(WidthBase=@n11, WidthTip=@n11, UVs for Projection=@n13, UVs for Thickness=@n13, UVs for texturing=@n13, UVs For NormalShading=@n13) {materialFunction=&SplineThicken}
  NamedRerouteDeclaration(None=@n14.WorldPositionOffset) {name=WPO, nodeColor=<rgba 0 0.7187 1 1>}
  n15 = Transform(None=@n14.Normal) {transformSourceType=TRANSFORMSOURCE_World, transformType=TRANSFORM_Tangent}
  n16 = TextureSample(UVs=Multiply(A=TextureCoordinate(), B=ScalarParameter() {defaultValue=1.5, parameterName=NormalTile})) {samplerType=SAMPLERTYPE_Normal, texture=/BRCosmetics/Accessories/FORT_Backpacks/Backpack_M_MED_GuineaPig/Textures/T_GuineaPig_WireEnds_N}
  NamedRerouteDeclaration(None=MaterialFunctionCall(BaseNormal=NamedRerouteDeclaration(None=@n15) {name=WireNormal, nodeColor=<rgba 1 0.5625 0 1>}, AdditionalNormal=@n16.RGB) {materialFunction=&BlendAngleCorrectedNormals}.Result) {name=WireEndNormal, nodeColor=<rgba 1 0.6328 0 1>}
  n17 = ConstantBiasScale(None=ComponentMask(None=@n1) {r=true}) {bias=-0.5, scale=2}
  NamedRerouteDeclaration(None=Step(Y=Abs(None=@n17), X=ScalarParameter() {defaultValue=0.4, parameterName=WireEndWidthMultiplier}) {constX=0.5, constY=0.5}) {name=WireEndOpacity, nodeColor=<rgba 0.9219 1 0 1>}
  n18 = SetMaterialAttributes!(
   MaterialAttributes
   Base Color=VectorParameter() {parameterName=InnerWireColor, defaultValue=<rgba 0.4115 0.1058 0.0357 1>}.RGB
   Metallic=Constant() {r=1}
   Roughness=ScalarParameter() {defaultValue=0.3, parameterName=InnerWireRoughness}
   Normal=NamedRerouteUsage()
   Opacity Mask=NamedRerouteUsage()
  ) {attributeSetTypes [
    #1
    #2
    #3
    #4
    #5
   ]}
  NamedRerouteDeclaration(None=@n18) {name=WireEndMat, nodeColor=<rgba 1 0 0.5859 1>}
  n19 = SetMaterialAttributes(
   Base Color=NamedRerouteUsage()
   Metallic=Constant()
   Roughness=NamedRerouteUsage()
   Normal=NamedRerouteUsage()
   Opacity Mask=Constant() {r=1}
  ) {attributeSetTypes [
    #1
    #2
    #3
    #4
    #5
   ]}
  NamedRerouteDeclaration(None=@n19) {name=WireMat, nodeColor=<rgba 0.2656 0 1 1>}
  SetMaterialAttributes!(
   MaterialAttributes=BlendMaterialAttributes(A=NamedRerouteUsage(), B=NamedRerouteUsage(), Alpha=NamedRerouteUsage())
   World Position Offset=NamedRerouteUsage()
  ) {attributeSetTypes [#6]}
 ]
}
