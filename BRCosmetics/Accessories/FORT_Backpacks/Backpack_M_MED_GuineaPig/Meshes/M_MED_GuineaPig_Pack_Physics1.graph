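# Physics asset for the MED GuineaPig backpack cosmetic, giving collision and joint setup to a small right-side bone chain.
# Capsule bodies sit on the pack, clavicle_r and shoulder_01_r bones, with no left-side bodies.
# The clavicle joint to the pack swings and twists freely, while the shoulder joint to the clavicle has a narrow swing limit and zero twist range.
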PhysicsAsset /BRCosmetics/Accessories/FORT_Backpacks/Backpack_M_MED_GuineaPig/Meshes/M_MED_GuineaPig_Pack_Physics1
bodies {
 clavicle_r [{
   shapeType=Capsule
   length=5.907
   radius=5.407
   shapeName=clavicle_r_capsule
   center=<3.9212 0.6013 0.4274>
   rotation=<rot 0 170.803 -152.2083>
  }]
 pack [{shapeType=Capsule, length=3, radius=3}]
 shoulder_01_r [{shapeType=Capsule, length=5.2782, radius=4.5415, shapeName=shoulder_01_r_capsule, center=<4.264 0 0>, rotation=<rot 90 0 0>}]
}
constraints [
 {bone1Name=shoulder_01_r, bone2Name=clavicle_r, swing1LimitDegrees=20, swing2LimitDegrees=20, twistLimitDegrees=0}
 {
  bone1Name=clavicle_r
  bone2Name=pack
  swing1LimitDegrees=45
  swing1Motion=Free
  swing2LimitDegrees=45
  swing2Motion=Free
  twistLimitDegrees=45
  twistMotion=Free
 }
]
